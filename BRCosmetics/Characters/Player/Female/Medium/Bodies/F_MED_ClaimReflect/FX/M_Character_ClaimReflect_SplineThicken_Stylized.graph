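# Translucent, unlit, two-sided material for stylized fire-textured Niagara ribbons and mesh particles in the ClaimReflect character effects.
# A panned, distorted, UV-shaped channel-packed texture blends three colors, with a Glow switch for single-color output.
# Opacity combines texture, dissolve along U and depth fade through SmoothStep for adjustable edge hardness.
# World position offset adds camera offset, a faded sine wave and SplineThicken width, driven by vertex color and dynamic parameters.
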Material /BRCosmetics/Characters/Player/Female/Medium/Bodies/F_MED_ClaimReflect/FX/M_Character_ClaimReflect_SplineThicken_Stylized
graph {
 properties {
  bTangentSpaceNormal=false
  bUsedWithNiagaraMeshParticles=true
  bUsedWithNiagaraRibbons=true
  blendMode=BLEND_Translucent
  pixelDepthOffsetMode=PDOM_Legacy
  shadingModel=MSM_Unlit
  twoSided=true
  usedShadingModels=Unlit
 }
 nodes [
  n1 = ScalarParameter() {group=WPO, parameterName=Camera offset}
  n2 = NamedRerouteUsage()
  n3 = DynamicParameter() {defaultValue=<rgba 1 1 0 0>, paramNames [WPO, Speed, WPO Disturbance, U offset]}
  n4 = ScalarParameter() {group=UV Manipulation, parameterName=U Offset, sliderMax=1}
  n5 = AppendVector(A=Add(A=Add(A=@n3.U offset, B=@n4), B=VertexColor().R), B=Constant())
  n6 = ScalarParameter() {defaultValue=-1, group=WPO, parameterName=WPO Wave Speed}
  n7 = VertexColor()
  n8 = StaticSwitchParameter(True=Add(A=Constant() {r=1}, B=Multiply(A=@n7.R) {constB=0.25}), False=Constant() {r=1}) {group=Main Texture, parameterName=Vertex Color Speed Control}
  n9 = Add(A=ComponentMask(None=Add(A=@n2, B=@n5)) {r=true}, B=Multiply(A=Time(), B=Multiply(A=@n6, B=@n8)) {constB=-1.3})
  n10 = ScalarParameter() {defaultValue=1, group=WPO, parameterName=WPO Wave Freq, sliderMax=10}
  n11 = Multiply(
   A=Sine(None=Multiply(A=@n9, B=@n10)) {period=1.2}
   B=VectorParameter() {group=WPO, parameterName=WPO Wave Direction, defaultValue=<rgba 0 0 1 1>}.RGB
  )
  n12 = Multiply(
   A=LinearInterpolate(
    A=Transform(None=@n11) {transformSourceType=TRANSFORMSOURCE_Camera}
    B=Transform(None=@n11) {transformSourceType=TRANSFORMSOURCE_Local}
    Alpha=ScalarParameter() {group=WPO, parameterName=WPO Change Transform, sliderMax=1}
   )
   B=Add(
    A=DynamicParameter() {defaultValue=<rgba 1 1 0 0>, paramNames [WPO, Speed, WPO Disturbance, U offset]}.WPO Disturbance
    B=ScalarParameter() {group=WPO, parameterName=WPO Disturbance, sliderMax=2}
   )
  )
  n13 = ComponentMask(None=NamedRerouteUsage()) {r=true}
  n14 = Power(Base=StaticSwitchParameter(True=OneMinus(None=@n13), False=@n13) {group=WPO, parameterName=Flip WPO Fade})
  n15 = ComponentMask(None=TextureCoordinate()) {r=true}
  n16 = LinearInterpolate(
   A=ScalarParameter() {defaultValue=12, group=Spline, parameterName=SplineThickenRoot, sliderMax=1}
   B=ScalarParameter() {defaultValue=12, group=Spline, parameterName=SplineThickenTip, sliderMax=1}
   Alpha=Sine(None=OneMinus(None=@n15)) {period=2}
  )
  n17 = TextureCoordinate()
  n18 = StaticBool() {desc=intended to switch with ribbons}
  Add(
   A=Add(
    A=MaterialFunctionCall(Offset Amount=@n1) {materialFunction=/Engine/Functions/Engine_MaterialFunctions02/WorldPositionOffset/CameraOffset}.Camera Offset
    B=Multiply(A=@n12, B=@n14)
   )
   B=Multiply(
    A=VertexColor().G
    B=MaterialFunctionCall(WidthBase=@n16, WidthTip=@n16, UVs for Projection=@n17, UVs for Thickness=@n17, Expand U or V UV Channel=@n18) {materialFunction=&SplineThicken}.WorldPositionOffset
   )
  )
  n19 = VectorParameter() {group=Color, parameterName=Color 1, defaultValue=<rgba 0 0 0 1>}
  n20 = ParticleColor()
  n21 = Multiply(A=@n20.A, B=@n19.A)
  n22 = NamedRerouteUsage()
  n23 = ComponentMask(None=MaterialFunctionCall(Texture Scale=ScalarParameter() {defaultValue=1, group=UV Manipulation, parameterName=UV End Pinch, sliderMax=1}) {materialFunction=&ScaleUVsByCenter}.UVs) {g=true}
  n24 = ComponentMask(None=@n22) {r=true}
  n25 = AppendVector(
   A=ComponentMask(None=@n22) {r=true}
   B=LinearInterpolate(A=ComponentMask(None=@n22) {g=true}, B=Saturate(None=@n23), Alpha=Power(Base=@n24) {constExponent=5})
  )
  n26 = Power(
   Base=ComponentMask(None=@n25) {r=true}
   Exp=ScalarParameter() {defaultValue=1, group=UV Manipulation, parameterName=UV U stretch, sliderMax=2}
  ) {constExponent=0.8}
  n27 = ComponentMask(None=@n25) {g=true}
  n28 = ScalarParameter() {defaultValue=1, group=UV Manipulation, parameterName=UV V Puff}
  n29 = LinearInterpolate(
   A=@n27
   B=ComponentMask(None=MaterialFunctionCall(Texture Scale=@n28) {materialFunction=&ScaleUVsByCenter}.UVs) {g=true}
   Alpha=Sine(None=@n27) {period=2}
  )
  n30 = Panner(
   Coordinate=NamedRerouteUsage()
   Speed=AppendVector(
    A=ScalarParameter() {defaultValue=1, group=Distortion, parameterName=Distortion Speed X, sliderMax=2, sliderMin=-2}
    B=ScalarParameter() {group=Distortion, parameterName=Distortion Speed Y, sliderMax=2, sliderMin=-2}
   )
  ) {speedX=-1}
  n31 = AppendVector(
   A=ScalarParameter() {defaultValue=1, group=Distortion, parameterName=Distortion Tiling X, sliderMax=2}
   B=ScalarParameter() {defaultValue=1, group=Distortion, parameterName=Distortion Tiling Y, sliderMax=2}
  )
  n32 = Subtract(A=TextureSampleParameter2D(UVs=Multiply(A=@n30, B=@n31) {constB=2}) {group=Distortion, parameterName=Distortion Texture, texture=&T_Inky_Smoke_Tile}.R) {constB=0.5}
  n33 = ScalarParameter() {group=Distortion, parameterName=Distortion Intensity, sliderMax=2, sliderMin=-2}
  n34 = AppendVector(
   A=ScalarParameter() {defaultValue=1, group=Main Texture, parameterName=Speed X, sliderMax=2, sliderMin=-2}
   B=ScalarParameter() {group=Main Texture, parameterName=Speed Y, sliderMax=2}
  )
  n35 = DynamicParameter() {parameterIndex=1, paramNames [SpeedMult,param2,param3,param4]}
  n36 = Add(A=Constant() {r=1}, B=Multiply(A=VertexColor().R))
  n37 = Constant() {r=1}
  n38 = Panner(
   Coordinate=Add(A=AppendVector(A=@n26, B=@n29), B=Multiply(A=@n32, B=@n33))
   Speed=Multiply(
    A=Multiply(A=@n34, B=@n35.SpeedMult) {constB=2}
    B=StaticSwitchParameter(True=@n36, False=@n37) {group=Main Texture, parameterName=Vertex Color Speed Control}
   )
  ) {speedX=-1}
  n39 = AppendVector(
   A=ScalarParameter() {defaultValue=1, group=Main Texture, parameterName=Tiling X, sliderMax=2}
   B=ScalarParameter() {defaultValue=1, group=Main Texture, parameterName=Tiling Y, sliderMax=2}
  )
  n40 = VertexColor()
  n41 = Constant()
  n42 = TextureSampleParameter2D(UVs=Add(A=Multiply(A=@n38, B=@n39) {constB=2}, B=AppendVector(A=@n40.R, B=@n41))) {
   desc=Channel packed texture
   group=Main Texture
   parameterName=Ribbon Texture
   texture=&/T_Character_ClaimReflect_Fire
  }
  n43 = Max(A=Max(A=@n42.R, B=@n42.R), B=@n42.B)
  n44 = LinearInterpolate(
   A=LinearInterpolate(
    A=VectorParameter() {group=Color, parameterName=Color 1, defaultValue=<rgba 0 0 0 1>}.A
    B=VectorParameter() {group=Color, parameterName=Color 2, defaultValue=<rgba 0.5 0.5 0.5 1>}.A
    Alpha=@n42.G
   )
   B=VectorParameter() {group=Color, parameterName=Color 3, defaultValue=<rgba 1 1 1 1>}.A
   Alpha=@n42.B
  )
  n45 = Multiply(A=NamedRerouteUsage()) {constB=2}
  n46 = TextureCoordinate()
  n47 = LinearInterpolate(
   A=ScalarParameter() {defaultValue=1, group=Dissolve, parameterName=Dissolve Min, sliderMax=10, sliderMin=-10}
   B=ScalarParameter() {group=Dissolve, parameterName=Dissolve Max, sliderMax=10, sliderMin=-10}
   Alpha=OneMinus(None=ComponentMask(None=@n46) {r=true})
  ) {constA=-2, constB=2}
  n48 = OneMinus(None=ComponentMask(None=TextureCoordinate()) {r=true})
  n49 = Multiply(
   A=StaticSwitchParameter(True=Multiply(A=@n21, B=@n42.A), False=Multiply(A=@n43, B=@n44)) {group=Glow, parameterName=Glow}
   B=Multiply(A=Subtract(A=@n45, B=@n47), B=Power(Base=@n48) {constExponent=1})
  )
  n50 = ScalarParameter() {defaultValue=10, group=Opacity, parameterName=FadeDistance, sliderMax=50}
  Multiply(
   A=@n20.A
   B=SmoothStep(
    Min=ScalarParameter() {group=Opacity, parameterName=Edge Hardness Min, sliderMax=1}
    Max=ScalarParameter() {defaultValue=0.9, group=Opacity, parameterName=Edge Hardness Max, sliderMax=1}
    Value=DepthFade(Opacity=@n49, FadeDistance=@n50)
   ) {constMax=0.1}
  )
  n51 = MaterialFunctionCall(Float2=NamedRerouteDeclaration(None=TextureCoordinate()) {name=Texcoord0 Unit, nodeColor=<rgba 0.0626 0.0553 0.8229 1>}) {materialFunction=&BreakOutFloat2Components}
  n52 = ScalarParameter() {defaultValue=2, parameterName=WPO Texture Scale X, sliderMax=5}
  NamedRerouteDeclaration(None=MaterialFunctionCall(X=Multiply(A=@n51.R, B=@n52), Y=@n51.G) {materialFunction=&MakeFloat2}.Result) {name=WPO TexCoord0 Scale, nodeColor=<rgba 0 0.5312 1 1>}
  NamedRerouteDeclaration(None=@n42.A) {name=Alpha Channel, nodeColor=<rgba 1 0.8672 0 1>}
  n53 = Multiply(A=@n20.RGB, B=@n19.RGB)
  n54 = VectorParameter() {group=Color, parameterName=Color 2, defaultValue=<rgba 0.5 0.5 0.5 1>}
  n55 = Multiply(
   A=LinearInterpolate(
    A=LinearInterpolate(A=@n19.RGB, B=@n54.RGB, Alpha=@n42.G)
    B=VectorParameter() {group=Color, parameterName=Color 3, defaultValue=<rgba 1 1 1 1>}.RGB
    Alpha=@n42.B
   )
   B=@n20.RGB
  )
  Reroute(None=MaterialFunctionCall(Emissive=StaticSwitchParameter(True=@n53, False=@n55) {group=Glow, parameterName=Glow}) {materialFunction=&MF_FX_EmissiveAdjustments}.Result)
 ]
}
