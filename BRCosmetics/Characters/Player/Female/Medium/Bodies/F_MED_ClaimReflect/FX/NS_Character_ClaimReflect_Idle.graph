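# Niagara idle effect for the ClaimReflect character body: head and neck fire that follows skeletal mesh bones, with orange and blue color variants.
# Burst-spawned mesh layers, a glow sprite and a light form the core, while continuous emitters add toon flames, wisps and sparks.
# Smoothed positions feed mesh particle attributes and material parameters, and the variant index sets mesh renderer visibility.
# Front-end switches ramp effects in by system age, and bisHeadHidden suppresses them.
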NiagaraSystem /BRCosmetics/Characters/Player/Female/Medium/Bodies/F_MED_ClaimReflect/FX/NS_Character_ClaimReflect_Idle
dependencies {
 usedDataInterfaces [
  &NiagaraDataInterfaceCurve
  &NiagaraDataInterfaceColorCurve
  &NiagaraDataInterfaceArrayFloat
  &NiagaraDataInterfaceSkeletalMesh
 ]
 usedDynamicInputs [&MakeVector2D, &SelectFloatFromArray]
 usedModules [
  &self:SystemSpawnScript.SystemScriptSource.SystemScriptGraph.NiagaraNodeAssignment_0.SetVariables_#1
  &SystemState
  &FN_VariantManager
 ]
}
emitters {
 DecalGlowAndLight {
  emitterUpdateScript [
   &EmitterState(
    Inactive Response=::NewEnumerator1
    Life Cycle Mode=::NewEnumerator1
    Loop Behavior=::NewEnumerator1
    Loop Duration Mode=::NewEnumerator1
    UseLoopDelay=false
   )
   &SpawnBurst_Instantaneous(Use Spawn Probability=false, UseLoopCountLimit=false) {module=SpawnBurst_Instantaneous001}
  ]
  particleSpawnScript [&InitializeParticle(Color Mode=::NewEnumerator0, Sprite Size Mode=::NewEnumerator3, Uniform Sprite Size=40, UsePositionOffset=false, Write Material Random=true) {module=InitializeParticle001}]
  particleUpdateScript [
   &ParticleState(Kill Particles When Lifetime Has Elapsed=false)
   &SkeletalMeshLocation(
    Bone / Socket Index=1
    Bone Sampling Mode=::NewEnumerator4
    MaskSpawnGroup=false
    Sampled Position Offset=<3 1 0>
    Skeletal Mesh=$System.Skeletal Mesh
   )
   /SetVariables(Particles.Color=&VariantColorSelector(Variant Amount=0, Original Color=<rgba 1 0.05 0 1>, Variant 1 Color=<rgba 0 0.4522 1 1>, Variant 2 Color=?)) {isSetParameters=true}
   &ScaleSpriteSize(Uniform Curve Sprite Scale=DataInterface() {propertyValues {Curve {keys [
        {arriveTangent=2.701, interpMode=RCIM_Cubic, leaveTangent=2.701, tangentMode=RCTM_User, time=0, value=0}
        {arriveTangent=0.0224, interpMode=RCIM_Cubic, leaveTangent=0.0224, tangentMode=RCTM_User, time=1, value=1}
       ]}}})
   &FN_Light_Attributes(
    Color=&MultiplyLinearColor(
     Linear Color=$Particles.Color
     Scale Factor=&MakeVector4_FromFloat(Float=&FN_FrontEndFloatModifier(
       Gameplay=1
       Front End=&FloatFromCurve(
        FloatCurve=DataInterface() {propertyValues {Curve {
           defaultValue=1
           keys [{interpMode=RCIM_Cubic, tangentMode=RCTM_User, time=0.3, value=0}, {interpMode=RCIM_Cubic, tangentMode=RCTM_User, time=2, value=1}]
          }}}
        CurveIndex=$Engine.System.Age
       )
      ))
    )
    Create Color Attribute=true
    Create Position Attribute=true
    Create Radius Attribute=true
    Position=&AddVectorToPosition(Position=$Particles.Position, Vector=<0 0 0>)
    Radius=&MakeCustomFloatFromBool(Bool=$User.bisHeadHidden, True Float=0, False Float=12.5)
    Scale Color=3
   )
   &ScaleColor(
    Color Value To Scale=$Particles.Color
    Scale RGB=&VectorFromFloat(Value=&FN_FrontEndFloatModifier(
      Gameplay=&MakeCustomFloatFromBool(Bool=$User.bisHeadHidden, True Float=0, False Float=0.5)
      Front End=&FloatFromCurve(
       FloatCurve=DataInterface() {propertyValues {Curve {
          defaultValue=1
          keys [{interpMode=RCIM_Cubic, tangentMode=RCTM_User, time=0.6, value=0}, {interpMode=RCIM_Cubic, tangentMode=RCTM_User, time=2, value=1}]
         }}}
       CurveIndex=$Engine.System.Age
       Scale Curve=&MakeCustomFloatFromBool(Bool=$User.bisHeadHidden, True Float=0, False Float=0.5)
      )
     ))
    ScaleA=false
    ScaleRGB=true
   )
  ]
  rendererClasses [&NiagaraLightRendererProperties, &NiagaraSpriteRendererProperties]
  renderers [
   {rendererClass=&NiagaraLightRendererProperties, rendererIndex=0}
   {rendererClass=&NiagaraSpriteRendererProperties, rendererIndex=1}
  ]
 }
 Fire_Head {
  emitterUpdateScript [
   &EmitterState(
    Inactive Response=::NewEnumerator1
    Life Cycle Mode=::NewEnumerator1
    Loop Behavior=::NewEnumerator1
    Loop Duration Mode=::NewEnumerator1
    UseLoopDelay=false
   )
   &SpawnBurst_Instantaneous(Use Spawn Probability=false, UseLoopCountLimit=false)
  ]
  particleSpawnScript [
   &InitializeParticle(
    Mesh Scale=&Multiply_VectorByFloat(Vector=<0.9 1 1.3>, Float=0.3)
    Mesh Scale Mode=::NewEnumerator4
    Position Offset=<-10 0 0>
    UsePositionOffset=true
    Write Material Random=true
   )
   &InitialMeshOrientation(Mesh Axis To Orient=<0 1 0>, Mesh Orientation Mode=::NewEnumerator2, Rotation=<0 -0.25 0>, Use Rotation Vector=true)
   &CameraOffset(Camera Offset Amount=-5)
   &DynamicMaterialParameters(
    Index 0 Param 1=1.2
    Module.Index0Float0Write=true
    Module.Index0Float1Write=true
    Module.Index0Float2Write=true
    Module.Index0Float3Write=true
   )
  ]
  particleUpdateScript [
   &ParticleState(Kill Particles When Lifetime Has Elapsed=false)
   /SetVariables(Particles.Color=/Niagara/DynamicInputs/LinearColor/MakeLinearColorFromVector4(Vector:Vector4f=&MakeVector4FromVectorAndFloat(
      "Vector (RGB)"=&FN_FrontEndVector3Modifier(
       Gameplay=&MakeVectorFromLinearColorRGB(Color=&VariantColorSelector(Variant Amount=0, Original Color=<rgba 1 0.5 0 1>, Variant 1 Color=<rgba 0 1 1 1>, Variant 2 Color=?))
       Front End=&MakeVectorFromLinearColorRGB(Color=&VariantColorSelector(Variant Amount=0, Original Color=<rgba 2 1.6833 0.1 1>, Variant 1 Color=<rgba 0 1 1 1>, Variant 2 Color=?))
      )
      "Float (Alpha)"=&FN_FrontEndFloatModifier(
       Gameplay=/BRCosmetics/Weapons/FORT_Melee/Pickaxe_GateHound/FX/VariantManager/VariantFloatSelector(
        Original Float:NiagaraFloat=0.5
        Variant 1 Float:NiagaraFloat=0.5
        Variant 2 Float:NiagaraFloat=?
        Variant 3 Float:NiagaraFloat=?
        Variant Amount:NiagaraInt32=0
       )
       Front End=&VariantFloatSelector(
        Original Float:NiagaraFloat=0.1
        Variant 1 Float:NiagaraFloat=0.35
        Variant 2 Float:NiagaraFloat=?
        Variant 3 Float:NiagaraFloat=?
        Variant Amount:NiagaraInt32=0
       )
      )
     ))) {isSetParameters=true}
   &ScaleMeshSize(Scale Factor=&Multiply_Vector(
     A=/BRCosmetics/Weapons/FORT_Melee/Pickaxe_GateHound/FX/VariantManager/VariantVectorSelector(
      Original Vector:Vector3f=<1 1 1>
      Variant 1 Vector:Vector3f=<1.05 1 1>
      Variant 2 Vector:Vector3f=?
      Variant 3 Vector:Vector3f=?
      Variant Amount:NiagaraInt32=0
     )
     B=&MakeVector(
      X=&FN_FrontEndFloatModifier(
       Gameplay=&MakeCustomFloatFromBool(Bool=$User.bisHeadHidden, True Float=0, False Float=1)
       Front End=&FloatFromCurve(
        FloatCurve=DataInterface() {propertyValues {Curve {
           defaultValue=1
           keys [
            {interpMode=RCIM_Cubic, tangentMode=RCTM_User, time=0, value=0}
            {arriveTangent=0.009, interpMode=RCIM_Cubic, leaveTangent=0.009, tangentMode=RCTM_User, time=1.5, value=1}
           ]
          }}}
        CurveIndex=$Engine.System.Age
        Scale Curve=&MakeCustomFloatFromBool(Bool=$User.bisHeadHidden, True Float=0, False Float=1)
       )
      )
      Y=1
      Z=&FN_FrontEndFloatModifier(
       Gameplay=&MakeCustomFloatFromBool(Bool=$User.bisHeadHidden, True Float=0, False Float=1)
       Front End=&FloatFromCurve(
        FloatCurve=DataInterface() {propertyValues {Curve {
           defaultValue=1
           keys [
            {interpMode=RCIM_Cubic, tangentMode=RCTM_User, time=0, value=0}
            {arriveTangent=0.0854, interpMode=RCIM_Cubic, leaveTangent=0.0854, tangentMode=RCTM_User, time=1.5, value=1}
           ]
          }}}
        CurveIndex=$Engine.System.Age
        Scale Curve=&MakeCustomFloatFromBool(Bool=$User.bisHeadHidden, True Float=0, False Float=1)
       )
      )
     )
    ))
   &SkeletalMeshLocation(
    Bone / Socket Index=1
    Bone Sampling Mode=::NewEnumerator4
    MaskSpawnGroup=false
    Orientation Sampling=::NewEnumerator1
    Sampled Position Offset=<-5 5 0>
    Skeletal Mesh=$System.Skeletal Mesh
   )
   &InterpolateOverTime(
    Rate Of Change=15
    Target Vector Value:Vector3f=&TransformPosition(Position=$Particles.Position, Source Space=::Local, Destination Space=::World)
    Variable Type=::NewEnumerator2
   )
   /SetVariables(Particles.InterpolateDY:Vector4f=&MakeFloat4FromVectorAndFloat("Vector (RGB)"=$StackContext.InterpolateOverTime.InterpolateOverTime_V3_.Moving Average, "Float (Alpha)"=1)) {isSetParameters=true}
   /SetVariables(Particles.VisibilityTag=&MakeIntFromFloat(Float=$System.VariantIndex)) {isSetParameters=true}
  ]
  rendererClasses [&NiagaraMeshRendererProperties]
  renderers [
   {rendererClass=&NiagaraMeshRendererProperties, rendererIndex=0}
   {rendererClass=&NiagaraMeshRendererProperties, rendererIndex=1}
  ]
 }
 Fire_Neck {
  emitterUpdateScript [
   &EmitterState(
    Inactive Response=::NewEnumerator1
    Life Cycle Mode=::NewEnumerator1
    Loop Behavior=::NewEnumerator1
    Loop Duration Mode=::NewEnumerator1
    UseLoopDelay=false
   )
   &SpawnBurst_Instantaneous(Use Spawn Probability=false, UseLoopCountLimit=false)
  ]
  particleSpawnScript [
   &InitializeParticle(Color=<rgba 1 1 1 0.2>, Mesh Scale Mode=::NewEnumerator4, UsePositionOffset=false, Write Material Random=true) {module=InitializeParticle001}
   &InitialMeshOrientation(Mesh Axis To Orient=<0 0 1>, Mesh Orientation Mode=::NewEnumerator2, Rotation=<0.02 0 0>, Use Rotation Vector=true)
   &CameraOffset(Camera Offset Amount=1)
  ]
  particleUpdateScript [
   &ParticleState(Kill Particles When Lifetime Has Elapsed=false, Let Infinitely Lived Particles Die When Emitter Deactivates=true)
   /SetVariables(Particles.VisibilityTag=&MakeIntFromFloat(Float=$System.VariantIndex)) {isSetParameters=true}
   &SkeletalMeshLocation(Bone Sampling Mode=::NewEnumerator4, MaskSpawnGroup=false, Sampled Position Offset=<13 3 0>, Skeletal Mesh=$System.Skeletal Mesh)
   /SetVariables(Particles.Color=&MakeLinearColorFromVectorAndFloat(
     "Vector (RGB)"=&MakeVectorFromLinearColorRGB(Color=&MultiplyLinearColor(
       Linear Color=&VariantColorSelector(Variant Amount=0, Original Color=<rgba 1 0.6001 0.221 1>, Variant 1 Color=<rgba 0 0.125 0.5 1>, Variant 2 Color=?)
       Scale Factor=&MakeVector4_FromFloat(Float=1)
      ))
     "Float (Alpha)"=0.3
    )) {isSetParameters=true}
   &ScaleColor(
    Color Value To Scale=$Particles.Color
    Scale Alpha=&FN_FrontEndFloatModifier(
     Gameplay=&MakeCustomFloatFromBool(Bool=$User.bisHeadHidden, True Float=0, False Float=1)
     Front End=&FloatFromCurve(
      FloatCurve=DataInterface() {propertyValues {Curve {
         defaultValue=1
         keys [{interpMode=RCIM_Cubic, tangentMode=RCTM_User, time=0.3, value=0}, {interpMode=RCIM_Cubic, tangentMode=RCTM_User, time=2, value=1}]
        }}}
      CurveIndex=$Engine.System.Age
      Scale Curve=&MakeCustomFloatFromBool(Bool=$User.bisHeadHidden, True Float=0, False Float=1)
     )
    )
    Scale RGB=&Multiply_VectorByFloat(Vector=<1 1 1>, Float=&FN_FrontEndFloatModifier(Gameplay=1, Front End=1))
    ScaleA=true
    ScaleRGB=true
   )
   &InterpolateOverTime(Rate Of Change=20, Target Vector Value:NiagaraPosition=$Particles.Position, Variable Type=::NewEnumerator2) {module=InterpolateOverTime001}
   &DynamicMaterialParameters(
    Index 0 Param 2=1
    Index 0 Param 3=1
    Index 0 Param 4=1
    Index 1 Vector 4:Vector4f=&MakeFloat4FromVectorAndFloat("Vector (RGB)"=$Particles.InterpolateOverTime001.InterpolateOverTime_V3_.Moving Average, "Float (Alpha)"=0.8)
    Index 3 Vector 1:NiagaraPosition=$Particles.SkeletalMeshLocation.SampledPosition
    Module.Index0Float0Write=true
    Module.Index0Float1Write=true
    Module.Index0Float2Write=true
    Module.Index0Float3Write=true
    Module.Index3Float3Write=true
    Module.Index3VectorWrite=true
    Write Parameter Index 1=::NewEnumerator4
    Write Parameter Index 3=::NewEnumerator3
   )
   &SolveForcesAndVelocity(Clamp Velocity=false, Limit Acceleration=false)
  ]
  rendererClasses [&NiagaraMeshRendererProperties]
  renderers [
   {rendererClass=&NiagaraMeshRendererProperties, rendererIndex=0}
   {rendererClass=&NiagaraMeshRendererProperties, rendererIndex=1}
  ]
 }
 Fire_Spline {
  emitterUpdateScript [
   &EmitterState(
    Inactive Response=::NewEnumerator1
    Life Cycle Mode=::NewEnumerator1
    Loop Behavior=::NewEnumerator1
    Loop Duration Mode=::NewEnumerator1
    UseLoopDelay=false
   )
   &SpawnBurst_Instantaneous(Use Spawn Probability=false, UseLoopCountLimit=false)
  ]
  particleSpawnScript [
   &InitializeParticle(Lifetime=3, Mesh Scale Mode=::NewEnumerator4, UsePositionOffset=false, Write Material Random=true)
   &CameraOffset(Camera Offset Amount=5)
  ]
  particleUpdateScript [
   &ParticleState(
    Kill Particles When Lifetime Has Elapsed=false
    Let Infinitely Lived Particles Die When Emitter Deactivates=true
    Loop Particles Lifetime=true
   )
   &ScaleMeshSize(Scale Factor=&MakeVector(
     X=&Lerp_Float(
      A=&FloatFromCurve(
       FloatCurve=DataInterface() {propertyValues {Curve {
          defaultValue=1
          keys [{interpMode=RCIM_Cubic, tangentMode=RCTM_User, time=1, value=0.4}, {interpMode=RCIM_Cubic, tangentMode=RCTM_User, time=2, value=0.8}]
         }}}
       CurveIndex=$Engine.System.Age
      )
      B=1
      Alpha=&FloatFromCurve(FloatCurve=DataInterface() {propertyValues {Curve {
          defaultValue=1
          keys [
           {interpMode=RCIM_Cubic, tangentMode=RCTM_User, time=0, value=0}
           {interpMode=RCIM_Cubic, tangentMode=RCTM_User, time=0.4, value=1}
           {tangentMode=RCTM_User, time=1, value=0}
          ]
         }}})
     )
     Y=&Lerp_Float(
      A=0.9
      B=1
      Alpha=&FloatFromCurve(FloatCurve=DataInterface() {propertyValues {Curve {
          defaultValue=1
          keys [
           {interpMode=RCIM_Cubic, tangentMode=RCTM_User, time=0, value=0}
           {interpMode=RCIM_Cubic, tangentMode=RCTM_User, time=0.7, value=1}
           {tangentMode=RCTM_User, time=1, value=0}
          ]
         }}})
     )
     Z=&Lerp_Float(
      A=0.9
      B=1.1
      Alpha=&FloatFromCurve(FloatCurve=DataInterface() {propertyValues {Curve {
          defaultValue=1
          keys [
           {interpMode=RCIM_Cubic, tangentMode=RCTM_User, time=0, value=0}
           {interpMode=RCIM_Cubic, tangentMode=RCTM_User, time=0.2, value=1}
           {tangentMode=RCTM_User, time=1, value=0}
          ]
         }}})
     )
    ))
   &SkeletalMeshLocation(Bone Sampling Mode=::NewEnumerator4, MaskSpawnGroup=false, Sampled Position Offset=<2 0 0>, Skeletal Mesh=$System.Skeletal Mesh)
   &DynamicMaterialParameters(
    Index 1 Param 1=1
    Module.Index0Float0Write=true
    Module.Index0Float1Write=true
    Module.Index0Float2Write=true
    Module.Index0Float3Write=true
    Module.Index1Float0Write=true
    Module.Index1Float1Write=true
    Module.Index1Float2Write=true
    Module.Index1Float3Write=true
    Write Parameter Index 1=::NewEnumerator1
   )
   /SetVariables(Particles.VisibilityTag=&MakeIntFromFloat(Float=$System.VariantIndex)) {isSetParameters=true}
   &ScaleColor(
    Scale Alpha=&FN_FrontEndFloatModifier(
     Gameplay=&MakeCustomFloatFromBool(Bool=$User.bisHeadHidden, True Float=0, False Float=1)
     Front End=&FloatFromCurve(
      FloatCurve=DataInterface() {propertyValues {Curve {
         defaultValue=1
         keys [{interpMode=RCIM_Cubic, tangentMode=RCTM_User, time=0.5, value=0}, {interpMode=RCIM_Cubic, tangentMode=RCTM_User, time=2, value=1}]
        }}}
      CurveIndex=$Engine.System.Age
      Scale Curve=&MakeCustomFloatFromBool(Bool=$User.bisHeadHidden, True Float=0, False Float=1)
     )
    )
    Scale RGB=&Multiply_VectorByFloat(Vector=<1 1 1>, Float=&FN_FrontEndFloatModifier(Gameplay=1, Front End=1))
    ScaleA=true
    ScaleRGB=true
   )
  ]
  rendererClasses [&NiagaraMeshRendererProperties]
  renderers [
   {rendererClass=&NiagaraMeshRendererProperties, rendererIndex=0}
   {rendererClass=&NiagaraMeshRendererProperties, rendererIndex=1}
   {rendererClass=&NiagaraMeshRendererProperties, rendererIndex=2}
   {rendererClass=&NiagaraMeshRendererProperties, rendererIndex=3}
  ]
 }
 Flame_Sprites_Toon {
  emitterUpdateScript [
   &EmitterState(Life Cycle Mode=::NewEnumerator1, Loop Behavior=::NewEnumerator1, Loop Delay=1.1, Loop Duration Mode=::NewEnumerator1, UseLoopDelay=false)
   &SpawnRate(
    Spawn Probability=0.8
    SpawnRate=&FN_FrontEndFloatModifier(
     Gameplay=&MakeCustomFloatFromBool(Bool=$User.bisHeadHidden, True Float=0, False Float=25)
     Front End=&FloatFromCurve(
      FloatCurve=DataInterface() {propertyValues {Curve {
         defaultValue=1
         keys [{interpMode=RCIM_Cubic, tangentMode=RCTM_User, time=0.3, value=0}, {interpMode=RCIM_Cubic, tangentMode=RCTM_User, time=2, value=1}]
        }}}
      CurveIndex=$Engine.System.Age
      Scale Curve=&MakeCustomFloatFromBool(Bool=$User.bisHeadHidden, True Float=0, False Float=25)
     )
    )
    Use Spawn Probability=false
   )
  ]
  particleSpawnScript [
   &InitializeParticle(
    Color=&VariantColorSelector(
     Variant Amount=0
     Original Color=&RandomRangeLinearColor(Minimum=<rgba 0.5 0 0 1>, Maximum=<rgba 1.5 0.25 0 1>)
     Variant 1 Color=&RandomRangeLinearColor(Minimum=<rgba 0 0.0957 0.5 1>, Maximum=<rgba 0 0.6 1.5 1>)
     Variant 2 Color=?
    )
    Lifetime Max=1.2
    Lifetime Mode=::NewEnumerator1
    Position Mode=::NewEnumerator0
    Sprite Rotation Mode=::NewEnumerator2
    Sprite Size Mode=::NewEnumerator1
    Sprite UV Mode=::NewEnumerator3
    Uniform Sprite Size Max=15
    Uniform Sprite Size Min=12
    Write Material Random=true
   ) {module=InitializeParticle001}
   &SkeletalMeshLocation(
    Bone Sampling Mode=::NewEnumerator4
    Interpolate Mesh Sampling=false
    MaskSpawnGroup=false
    Orientation Sampling=::NewEnumerator1
    Sampled Position Offset=<10.8 0 0>
    Skeletal Mesh=$System.Skeletal Mesh
   )
   &ShapeLocation(
    Cylinder Distribution=::NewEnumerator2
    Cylinder Height=8
    Cylinder Height Midpoint=0
    Cylinder Radius=8
    MaskSpawnGroup=false
    Non Uniform Scale=<0.6 1 1>
    Rotation Axis=<0 0.4 1>
    Rotation Mode=::NewEnumerator1
    Shape Primitive=::NewEnumerator1
    U Position=&RandomRangeFloat(Minimum=0, Maximum=1)
    UseOwnerScale=true
    V Position=&RandomRangeFloat(Minimum=0, Maximum=1)
   )
   &AddVelocity(
    Module.Use Falloff Distance=false
    Origin Offset=<0 5 10>
    Override Seed=false
    Velocity Mode=::NewEnumerator1
    Velocity Origin=$Particles.SkeletalMeshLocation.SampledPosition
    Velocity Speed=&RandomRangeFloat(Minimum=10, Maximum=15)
   )
   &AddVelocity(
    Rotation Coordinate Space=::Simulation
    Velocity=&RotateVectorByQuaternion(Vector To Rotate=<2 1 0>, Quaternion=$Particles.SkeletalMeshLocation.SampledRotation)
    Velocity Speed Scale=&RandomRangeFloat(Minimum=5, Maximum=10)
   ) {module=AddVelocity001}
   &DynamicMaterialParameters(
    Index 0 Param 1=&RandomRangeFloat(Minimum=0, Maximum=1)
    Module.Index0Float0Write=true
    Module.Index0Float1Write=true
    Module.Index0Float2Write=true
    Module.Index0Float3Write=true
   ) {module=DynamicMaterialParameters001}
  ]
  particleUpdateScript [
   &ParticleState()
   &CameraOffset(Camera Offset Amount=&FloatFromCurve(
     FloatCurve=DataInterface() {propertyValues {Curve {
        defaultValue=1
        keys [
         {arriveTangent=0.0331, interpMode=RCIM_Cubic, leaveTangent=0.0331, tangentMode=RCTM_User, time=0, value=0}
         {interpMode=RCIM_Cubic, tangentMode=RCTM_User, time=0.5, value=1}
        ]
       }}}
     Scale Curve=15
    ))
   /SetVariables(Particles.SubImageIndex=&FloatFromCurve(FloatCurve=DataInterface() {propertyValues {Curve {keys [<at 0 0>,<at 1 15>]}}})) {isSetParameters=true}
   &SkeletalMeshLocation(
    Bone Sampling Mode=::NewEnumerator4
    Initialize Particle Position To Sampled Position=false
    MaskSpawnGroup=false
    Orientation Sampling=::NewEnumerator1
    Position Sampling=::NewEnumerator1
    Sampled Position Offset=<10.8 0 0>
    Skeletal Mesh=$System.Skeletal Mesh
    Source Movement Scale:Vector3f=&VectorFromFloat(Value=0.9)
   ) {module=SkeletalMeshLocation001}
   &InheritSourceMovement(
    Applied Position Scale=&VectorFromFloat(Value=&FloatFromCurve(FloatCurve=DataInterface() {propertyValues {Curve {
         defaultValue=1
         keys [
          {interpMode=RCIM_Cubic, tangentMode=RCTM_User, time=0.5, value=1}
          {arriveTangent=-2.0414, interpMode=RCIM_Cubic, leaveTangent=-2.0414, tangentMode=RCTM_User, time=1, value=0.5}
         ]
        }}}))
    ApplyForce=false
    ApplyPosition=true
    ApplyPositionFalloff=false
    Source Velocity:Vector3f=$Particles.SkeletalMeshLocation001.SampledVelocity
   ) {enabled=false}
   &ScaleSpriteSize(Scale Factor=&Vector2DFromFloat(Value=&FloatFromCurve(FloatCurve=DataInterface() {propertyValues {Curve {
         defaultValue=0
         keys [
          {interpMode=RCIM_Cubic, tangentMode=RCTM_User, time=0.5, value=1}
          {arriveTangent=-3.1709, interpMode=RCIM_Cubic, leaveTangent=-3.1709, tangentMode=RCTM_User, time=1, value=0}
         ]
        }}})))
   &ScaleColor(
    Scale Alpha=&Multiply_Float(
     A=&FloatFromCurve(
      FloatCurve=DataInterface() {propertyValues {Curve {keys [
          {arriveTangent=9.0725, interpMode=RCIM_Cubic, leaveTangent=9.0725, tangentMode=RCTM_User, time=0, value=0}
          {arriveTangent=-0.0135, interpMode=RCIM_Cubic, leaveTangent=-0.0135, tangentMode=RCTM_User, time=0.1408, value=0.9847}
          {arriveTangent=-2.251, interpMode=RCIM_Cubic, leaveTangent=-2.251, tangentMode=RCTM_User, time=1, value=0}
         ]}}}
      Scale Curve=&FN_FrontEndFloatModifier(Gameplay=1, Front End=1)
     )
     B=&RandomRangeFloat(Minimum=0.1, Maximum=0.2)
    )
    Scale RGB=&VectorFromFloat(Value=&FloatFromCurve(FloatCurve=DataInterface() {propertyValues {Curve {
         defaultValue=1
         keys [{interpMode=RCIM_Cubic, tangentMode=RCTM_User, time=0, value=1}, {interpMode=RCIM_Cubic, tangentMode=RCTM_User, time=0.8, value=0}]
        }}}))
    ScaleA=true
    ScaleRGB=true
   ) {module=ScaleColor001}
   &Drag(
    Drag=&FloatFromCurve(
     FloatCurve=DataInterface() {propertyValues {Curve {
        defaultValue=1
        keys [
         {arriveTangent=1.8143, interpMode=RCIM_Cubic, leaveTangent=1.8143, tangentMode=RCTM_User, time=0, value=0}
         {interpMode=RCIM_Cubic, tangentMode=RCTM_User, time=1, value=1}
        ]
       }}}
     Scale Curve=10
    )
    Use Linear Drag=true
    Use Rotational Drag=true
   )
   &CurlNoiseForce(Noise Strength=100, Pan Noise=true, Pan Noise Field=<0 0 0.1>)
   &GravityForce(Coordinate Space=::Local, Gravity=<0 0 20>)
   &SolveForcesAndVelocity(Clamp Velocity=false, Limit Acceleration=false)
  ]
  rendererClasses [&NiagaraSpriteRendererProperties]
  renderers [{rendererClass=&NiagaraSpriteRendererProperties}]
 }
 Flame_Sprites_Wispy {
  emitterUpdateScript [
   &EmitterState(
    Inactive Response=::NewEnumerator1
    Life Cycle Mode=::NewEnumerator1
    Loop Behavior=::NewEnumerator1
    Loop Duration Mode=::NewEnumerator1
    UseLoopDelay=false
   )
   &SpawnRate(
    SpawnRate=&FN_FrontEndFloatModifier(
     Gameplay=&MakeCustomFloatFromBool(Bool=$User.bisHeadHidden, True Float=0, False Float=30)
     Front End=&FloatFromCurve(
      FloatCurve=DataInterface() {propertyValues {Curve {
         defaultValue=1
         keys [{interpMode=RCIM_Cubic, tangentMode=RCTM_User, time=0.3, value=0}, {interpMode=RCIM_Cubic, tangentMode=RCTM_User, time=2, value=1}]
        }}}
      CurveIndex=$Engine.System.Age
      Scale Curve=&MakeCustomFloatFromBool(Bool=$User.bisHeadHidden, True Float=0, False Float=30)
     )
    )
    Use Spawn Probability=false
   )
  ]
  particleSpawnScript [
   &InitializeParticle(
    Lifetime Max=1
    Lifetime Min=0.8
    Lifetime Mode=::NewEnumerator1
    Position Offset=&RandomRangeVector(Minimum=<10 0 0>, Maximum=<5 0 0>)
    Sprite Rotation Mode=::NewEnumerator2
    Sprite Size Mode=::NewEnumerator1
    Uniform Sprite Size Max=12
    Uniform Sprite Size Min=7
    UsePositionOffset=true
    Write Material Random=true
   )
   &DynamicMaterialParameters(
    Index 0 Param 1=&RandomRangeFloat(Minimum=0, Maximum=1)
    Module.Index0Float0Write=true
    Module.Index0Float1Write=true
    Module.Index0Float2Write=true
    Module.Index0Float3Write=true
   )
   &SkeletalMeshLocation(
    Bone Sampling Mode=::NewEnumerator4
    Interpolate Mesh Sampling=false
    MaskSpawnGroup=false
    Orientation Sampling=::NewEnumerator1
    Sampled Position Offset=<10.8 0 0>
    Skeletal Mesh=$System.Skeletal Mesh
   )
   &ShapeLocation(
    Cylinder Distribution=::NewEnumerator2
    Cylinder Height=5
    Cylinder Height Midpoint=0
    Cylinder Radius=5
    MaskSpawnGroup=false
    Non Uniform Scale=<0.6 1 1>
    Rotation Axis=<0 0.4 1>
    Rotation Mode=::NewEnumerator1
    Shape Primitive=::NewEnumerator1
    U Position=&RandomRangeFloat(Minimum=0, Maximum=1)
    UseOwnerScale=true
    V Position=&RandomRangeFloat(Minimum=0, Maximum=1)
   )
   &AddVelocity(
    Module.Use Falloff Distance=false
    Origin Offset=<0 0 12>
    Override Seed=false
    Velocity Mode=::NewEnumerator1
    Velocity Origin=$Particles.SkeletalMeshLocation.SampledPosition
    Velocity Speed=&RandomRangeFloat(Minimum=10, Maximum=20)
   )
   &AddVelocity(
    Rotation Coordinate Space=::Simulation
    Velocity=&RotateVectorByQuaternion(Vector To Rotate=<2 1 0>, Quaternion=$Particles.SkeletalMeshLocation.SampledRotation)
    Velocity Speed Scale=&RandomRangeFloat(Minimum=10, Maximum=15)
   ) {module=AddVelocity001}
  ]
  particleUpdateScript [
   &ParticleState()
   &CameraOffset(Camera Offset Amount=&FloatFromCurve(
     FloatCurve=DataInterface() {propertyValues {Curve {
        defaultValue=1
        keys [
         {arriveTangent=0.0331, interpMode=RCIM_Cubic, leaveTangent=0.0331, tangentMode=RCTM_User, time=0, value=0.2}
         {interpMode=RCIM_Cubic, tangentMode=RCTM_User, time=0.5, value=1}
        ]
       }}}
     Scale Curve=20
    ))
   &SkeletalMeshLocation(
    Bone Sampling Mode=::NewEnumerator4
    Initialize Particle Position To Sampled Position=false
    MaskSpawnGroup=false
    Orientation Sampling=::NewEnumerator1
    Position Sampling=::NewEnumerator1
    Sampled Position Offset=<10.8 0 0>
    Skeletal Mesh=$System.Skeletal Mesh
    Source Movement Scale:Vector3f=&VectorFromFloat(Value=0.9)
   ) {module=SkeletalMeshLocation001}
   &InheritSourceMovement(
    Applied Position Scale=&VectorFromFloat(Value=&FloatFromCurve(FloatCurve=DataInterface() {propertyValues {Curve {defaultValue=1, keys [{interpMode=RCIM_Cubic, tangentMode=RCTM_User, time=0.5, value=1}]}}}))
    ApplyForce=false
    ApplyPosition=true
    ApplyPositionFalloff=false
   ) {enabled=false}
   &ScaleSpriteSize(Scale Factor=&Vector2DFromFloat(Value=&FloatFromCurve(FloatCurve=DataInterface() {propertyValues {Curve {
         defaultValue=1
         keys [{interpMode=RCIM_Cubic, tangentMode=RCTM_User, time=0, value=1}, {interpMode=RCIM_Cubic, tangentMode=RCTM_User, time=1, value=1.5}]
        }}})))
   &ScaleColor(
    Scale Alpha=&Multiply_Float(
     A=&FloatFromCurve(
      FloatCurve=DataInterface() {propertyValues {Curve {
         defaultValue=0
         keys [
          {arriveTangent=15.019, interpMode=RCIM_Cubic, leaveTangent=15.019, tangentMode=RCTM_User, time=0, value=0}
          {interpMode=RCIM_Cubic, tangentMode=RCTM_User, time=0.2, value=1}
          {tangentMode=RCTM_User, time=1, value=0}
         ]
        }}}
      Scale Curve=&FN_FrontEndFloatModifier(Gameplay=1, Front End=1)
     )
     B=&RandomRangeFloat(Minimum=2, Maximum=3)
    )
    Scale RGB=&MakeVectorFromLinearColorRGB(Color=&VariantColorSelector(
      Variant Amount=0
      Original Color=&ColorFromCurve(ColorCurve=DataInterface() {propertyValues {
         AlphaCurve {keys [<at 0 1 RCIM_Cubic>]}
         BlueCurve {keys [<at 0 0>,<at 0.3657 0 RCIM_Cubic>,<at 1 0 RCIM_Cubic>]}
         GreenCurve {keys [<at 0 0.3333>, {arriveTangent=-0.3083, interpMode=RCIM_Cubic, leaveTangent=-0.3083, time=0.3657, value=0.125}, <at 1 0.025 RCIM_Cubic>]}
         RedCurve {keys [<at 0 2>, {arriveTangent=-1.7, interpMode=RCIM_Cubic, leaveTangent=-1.7, time=0.3657, value=1.5}, <at 1 0.3 RCIM_Cubic>]}
        }})
      Variant 1 Color=&ColorFromCurve(ColorCurve=DataInterface() {propertyValues {
         AlphaCurve {keys [<at 0 1 RCIM_Cubic>]}
         BlueCurve {keys [<at 0 1.5>, {interpMode=RCIM_Cubic, leaveTangent=-1.2284, time=0.3781, value=1}, <at 1 0.5 RCIM_Cubic>]}
         GreenCurve {keys [<at 0 1>, {interpMode=RCIM_Cubic, leaveTangent=-1.1109, time=0.3781, value=0.4}, <at 1 0.0957 RCIM_Cubic>]}
         RedCurve {keys [<at 0 0>,<at 0.3781 0 RCIM_Cubic>,<at 1 0 RCIM_Cubic>]}
        }})
      Variant 2 Color=?
     ))
    ScaleA=true
    ScaleRGB=true
   ) {module=ScaleColor001}
   &Drag(
    Drag=&FloatFromCurve(
     FloatCurve=DataInterface() {propertyValues {Curve {
        defaultValue=1
        keys [
         {arriveTangent=1.8143, interpMode=RCIM_Cubic, leaveTangent=1.8143, tangentMode=RCTM_User, time=0, value=0}
         {interpMode=RCIM_Cubic, tangentMode=RCTM_User, time=1, value=1}
        ]
       }}}
     Scale Curve=10
    )
    Use Linear Drag=true
    Use Rotational Drag=true
   )
   &CurlNoiseForce(Noise Frequency=100, Noise Strength=100, Pan Noise=true, Pan Noise Field=<0 0 0.1>)
   &GravityForce(Coordinate Space=::Local, Gravity=<0 0 60>)
   &SolveForcesAndVelocity(Clamp Velocity=false, Limit Acceleration=false)
  ]
  rendererClasses [&NiagaraSpriteRendererProperties]
  renderers [{rendererClass=&NiagaraSpriteRendererProperties}]
 }
 Sparks {
  emitterUpdateScript [
   &InterpolateOverTime(
    Rate Of Change=3
    Target Float Value:NiagaraFloat=&MakeCustomFloatFromBool(
     Bool=&SetBoolByFloatComparison(A=&VectorLength(Vector=$Engine.Owner.Velocity), B=100)
     True Float=0
     False Float=1
    )
   )
   &EmitterState(
    Inactive Response=::NewEnumerator1
    Life Cycle Mode=::NewEnumerator1
    Loop Behavior=::NewEnumerator1
    Loop Duration Mode=::NewEnumerator1
    UseLoopDelay=false
   )
   &SpawnRate(
    SpawnRate=&FN_FrontEndFloatModifier(
     Gameplay=&MakeCustomFloatFromBool(Bool=$User.bisHeadHidden, True Float=0, False Float=50)
     Front End=&FloatFromCurve(
      FloatCurve=DataInterface() {propertyValues {Curve {
         defaultValue=1
         keys [{interpMode=RCIM_Cubic, tangentMode=RCTM_User, time=0.3, value=0}, {interpMode=RCIM_Cubic, tangentMode=RCTM_User, time=2, value=1}]
        }}}
      CurveIndex=$Engine.System.Age
      Scale Curve=&MakeCustomFloatFromBool(Bool=$User.bisHeadHidden, True Float=0, False Float=50)
     )
    )
    Use Spawn Probability=false
   )
  ]
  particleSpawnScript [
   &InitializeParticle(
    Color=&VariantColorSelector(Variant Amount=0, Original Color=<rgba 3 0.7919 0.3503 1>, Variant 1 Color=<rgba 0.2335 1.0323 2 1>, Variant 2 Color=?)
    Lifetime Max=0.8
    Lifetime Min=0.7
    Lifetime Mode=::NewEnumerator1
    Sprite Size Max=<1 5>
    Sprite Size Min=<0.75 4>
    Sprite Size Mode=::NewEnumerator2
    UsePositionOffset=false
    Write Material Random=true
   )
   &SkeletalMeshLocation(
    Bone Sampling Mode=::NewEnumerator4
    Interpolate Mesh Sampling=false
    MaskSpawnGroup=false
    Sampled Position Offset=<12 1 0>
    Skeletal Mesh=$System.Skeletal Mesh
   )
   &ShapeLocation(
    Cylinder Distribution=::NewEnumerator2
    Cylinder Height=1
    Cylinder Height Midpoint=0
    Cylinder Radius=5
    MaskSpawnGroup=false
    Non Uniform Scale=<1 1.5 1>
    Rotation Axis=<0 0.6 1>
    Rotation Mode=::NewEnumerator1
    Shape Primitive=::NewEnumerator1
    U Position=&RandomRangeFloat(Minimum=0, Maximum=1)
    UseOwnerScale=true
    V Position=&RandomRangeFloat(Minimum=0, Maximum=1)
   ) {module=ShapeLocation001}
   &AddVelocity(
    Module.Use Falloff Distance=false
    Origin Offset=<0 7 10>
    Override Seed=false
    Velocity Mode=::NewEnumerator1
    Velocity Origin=$Particles.SkeletalMeshLocation.SampledPosition
    Velocity Speed=&RandomRangeFloat(Minimum=35, Maximum=50)
   )
   &AddVelocity(
    Rotation Coordinate Space=::Simulation
    Velocity=&RotateVectorByQuaternion(Vector To Rotate=<1 -0.5 0>, Quaternion=$Particles.SkeletalMeshLocation.SampledRotation)
    Velocity Speed Scale=&RandomRangeFloat(Minimum=20, Maximum=30)
   ) {module=AddVelocity001}
   &CameraOffset(Camera Offset Amount=5)
  ]
  particleUpdateScript [
   &ParticleState()
   &SkeletalMeshLocation(
    Bone Sampling Mode=::NewEnumerator4
    Initialize Particle Position To Sampled Position=false
    MaskSpawnGroup=false
    Orientation Sampling=::NewEnumerator1
    Position Sampling=::NewEnumerator1
    Sampled Position Offset=<12 1 0>
    Skeletal Mesh=$System.Skeletal Mesh
    Source Movement Scale:Vector3f=&VectorFromFloat(Value=0.9)
   ) {module=SkeletalMeshLocation001}
   &InheritSourceMovement(
    Applied Position Scale=&VectorFromFloat(Value=&FloatFromCurve(FloatCurve=DataInterface() {propertyValues {Curve {
         defaultValue=1
         keys [
          {interpMode=RCIM_Cubic, tangentMode=RCTM_User, time=0.5, value=1}
          {arriveTangent=-3.6498, interpMode=RCIM_Cubic, leaveTangent=-3.6498, tangentMode=RCTM_User, time=1, value=0}
         ]
        }}}))
    ApplyForce=false
    ApplyPosition=true
    ApplyPositionFalloff=false
    Source Velocity:Vector3f=$Particles.SkeletalMeshLocation001.SampledVelocity
   ) {enabled=false}
   /SetVariables(Particles.Glow03_Color=&MakeLinearColorFromVectorAndFloat(
     "Vector (RGB)"=&MakeVectorFromLinearColorRGB(Color=&MultiplyLinearColor(
       Linear Color=&VariantColorSelector(Variant Amount=0, Original Color=<rgba 1 0.0833 0 0.2>, Variant 1 Color=<rgba 0 0.2435 1 0.2>, Variant 2 Color=?)
       Scale Factor=&MakeVector4_FromFloat(Float=1)
      ))
     "Float (Alpha)"=&Multiply_Float(
      A=0.2
      B=&MakeCustomFloatFromBool(
       Bool=&SetBoolByFloatComparison(A=&VectorLength(Vector=$Engine.Owner.Velocity), B=100)
       True Float=0
       False Float=1
      )
     )
    )) {isSetParameters=true}
   /SetVariables(Particles.Glow03_Size=&Multiply_Vector2D_ByFloat(
     Float=&FloatFromCurve(
      FloatCurve=DataInterface() {propertyValues {Curve {keys [
          <at 0 0 RCIM_Cubic>
          <at 0.2 1.3 RCIM_Cubic>
          {arriveTangent=-0.1423, interpMode=RCIM_Cubic, leaveTangent=-0.1423, tangentMode=RCTM_User, time=1, value=0}
         ]}}}
      Scale Curve=13
     )
     Vector2D=<0.2 0.7>
    )) {isSetParameters=true}
   &ScaleColor(
    Scale Alpha=&FloatFromCurve(FloatCurve=DataInterface() {propertyValues {Curve {
        defaultValue=1
        keys [
         {interpMode=RCIM_Cubic, tangentMode=RCTM_User, time=0, value=0}
         {interpMode=RCIM_Cubic, tangentMode=RCTM_User, time=0.5, value=1}
         {tangentMode=RCTM_User, time=1, value=0}
        ]
       }}})
    ScaleA=true
    ScaleRGB=true
   )
   &ScaleSpriteSize(Scale Factor=&Vector2DFromFloat(Value=&FloatFromCurve(
      FloatCurve=DataInterface() {propertyValues {Curve {keys [
          <at 0 0 RCIM_Cubic>
          {interpMode=RCIM_Cubic, leaveTangent=-3.0296, tangentMode=RCTM_Break, time=0.2, value=1.2}
          {arriveTangent=-2.8894, interpMode=RCIM_Cubic, leaveTangent=-2.8894, tangentMode=RCTM_User, time=1, value=0.1}
         ]}}}
      Scale Curve=0.4
     ))) {module=ScaleSpriteSize001}
   &CurlNoiseForce(Noise Frequency=100, Noise Strength=100, Pan Noise=true, Pan Noise Field=<0 0 0.5>, Random Seed=10)
   &Drag(Drag=3.9, Use Linear Drag=true, Use Rotational Drag=false)
   &GravityForce(Coordinate Space=::Local, Gravity=<0 -10 10>)
   &SolveForcesAndVelocity(Clamp Velocity=false, Limit Acceleration=false)
  ]
  rendererClasses [&NiagaraSpriteRendererProperties]
  renderers [
   {rendererClass=&NiagaraSpriteRendererProperties}
   {rendererClass=&NiagaraSpriteRendererProperties, rendererIndex=1}
  ]
 }
}
summary {
 systemName=NS_Character_ClaimReflect_Idle
 emitters [
  {emitterName=Flame_Sprites_Wispy, rendererClasses [&NiagaraSpriteRendererProperties]}
  {emitterName=Flame_Sprites_Toon, rendererClasses [&NiagaraSpriteRendererProperties]}
  {emitterName=Sparks, rendererClasses [&NiagaraSpriteRendererProperties]}
  {emitterName=Fire_Spline, rendererClasses [&NiagaraMeshRendererProperties]}
  {
   emitterName=DecalGlowAndLight
   rendererClasses [&NiagaraLightRendererProperties, &NiagaraSpriteRendererProperties]
  }
  {emitterName=Fire_Neck, rendererClasses [&NiagaraMeshRendererProperties]}
  {emitterName=Fire_Head, rendererClasses [&NiagaraMeshRendererProperties]}
 ]
 userVariables [
  User.bIsFrontEnd {type=NiagaraBool}
  User.bIsFrontEndPreview {type=NiagaraBool}
  User.bisHeadHidden {type=NiagaraBool}
  User.Variant {type=NiagaraFloat, defaultValue=0, description=""}
 ]
}
